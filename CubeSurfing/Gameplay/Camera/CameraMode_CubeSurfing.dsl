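; CubeSurfing camera mode that temporarily reshapes the camera origin's vertical interpolation speed when a jump starts.
; JumpStarted caches the interp speed and starts a timer lasting the Curve_ZInterpMult curve's length, skipping this if that timer is active.
; A gated UpdateZInterp reschedules itself each tick, scaling cached Z speed by the curve sampled at timer elapsed time.
; When the timer expires, ResetInterpSpeed restores the cached speed.
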(event Custom|JumpStarted
  (bind (_mintime _maxtime) (Math|Curves|GetTimeRange (Variables|Default|GetCurve_ZInterpMult)))
  (bind (_cachedinterpspeed_x _cachedinterpspeed_y _cachedinterpspeed_z) (Math|Vector|BreakVector (Variables|GetCachedInterpSpeed_X)))
  (if (not (Utilities|Time|IsTimerActivebyHandle (Variables|Default|GetZInterpTimer)))
    (Variables|Default|SetCachedInterpSpeed (Variables|CameraPivot|GetCameraOriginInterpSpeed))
    (bind _returnvalue (Utilities|Time|SetTimerbyEvent (delegate Custom|ResetInterpSpeed) _maxtime))
    (Variables|Default|SetZInterpTimer _returnvalue)
    (Utilities|FlowControl|Sequence
      (:then_0
        (Utilities|FlowControl|Gate true
          (:Exit
            (Variables|SetCameraOriginInterpSpeed_X (Math|Vector|MakeVector _cachedinterpspeed_x _cachedinterpspeed_y (* (Math|Curves|GetFloatValue (Variables|Default|GetCurve_ZInterpMult) (Utilities|Time|GetTimerElapsedTimebyHandle (Variables|Default|GetZInterpTimer))) _cachedinterpspeed_z)))
            (Utilities|Time|SetTimerforNextTickbyEvent (delegate Custom|UpdateZInterp)))))
      (:then_1
        (CallFunction|UpdateZInterp)))))

(event Custom|UpdateZInterp
  (bind (_cachedinterpspeed_x _cachedinterpspeed_y _cachedinterpspeed_z) (Math|Vector|BreakVector (Variables|GetCachedInterpSpeed_X)))
  (Utilities|FlowControl|Gate true
    (:Exit
      (Variables|SetCameraOriginInterpSpeed_X (Math|Vector|MakeVector _cachedinterpspeed_x _cachedinterpspeed_y (* (Math|Curves|GetFloatValue (Variables|Default|GetCurve_ZInterpMult) (Utilities|Time|GetTimerElapsedTimebyHandle (Variables|Default|GetZInterpTimer))) _cachedinterpspeed_z)))
      (Utilities|Time|SetTimerforNextTickbyEvent (delegate Custom|UpdateZInterp)))))

(event Custom|ResetInterpSpeed
  (bind (_cachedinterpspeed_x _cachedinterpspeed_y _cachedinterpspeed_z) (Math|Vector|BreakVector (Variables|GetCachedInterpSpeed_X)))
  (Variables|CameraPivot|SetCameraOriginInterpSpeed (Variables|Default|GetCachedInterpSpeed))
  (Utilities|FlowControl|Gate true
    (:Exit
      (Variables|SetCameraOriginInterpSpeed_X (Math|Vector|MakeVector _cachedinterpspeed_x _cachedinterpspeed_y (* (Math|Curves|GetFloatValue (Variables|Default|GetCurve_ZInterpMult) (Utilities|Time|GetTimerElapsedTimebyHandle (Variables|Default|GetZInterpTimer))) _cachedinterpspeed_z)))
      (Utilities|Time|SetTimerforNextTickbyEvent (delegate Custom|UpdateZInterp)))))
