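; Camera mode for boosting in CubeSurfing that animates camera-space forward distance along a decay curve.
; StartForwardDecay caches the current forward distance and server start time, then calls TickDecay.
; TickDecay sets the distance to the cached value multiplied by the curve sampled at elapsed time.
; It re-arms itself with a next-tick timer until elapsed time reaches the curve's end.
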(event Custom|StartForwardDecay
  (Variables|Default|SetCachedForward (Variables|CameraSettings|GetCameraSpaceForwardDistance))
  (Variables|Default|SetStartingTime (Fortnite|Sync|Time|GetServerWorldTimeSeconds))
  (CallFunction|TickDecay))

(event Custom|TickDecay
  (bind _forwarddecaycurve (Variables|Default|GetForwardDecayCurve))
  (bind (_mintime _maxtime) (Math|Curves|GetTimeRange _forwarddecaycurve))
  (Variables|CameraSettings|SetCameraSpaceForwardDistance (* (Variables|Default|GetCachedForward) (Math|Curves|GetFloatValue _forwarddecaycurve (CallFunction|GetElapsedTime))))
  (if (not (>= (CallFunction|GetElapsedTime) _maxtime))
    (Utilities|Time|SetTimerforNextTickbyEvent (delegate Custom|TickDecay))))

(fn GetElapsedTime
  (return (- (Fortnite|Sync|Time|GetServerWorldTimeSeconds) (Variables|Default|GetStartingTime))))
