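; Camera modifier for the Cube Surfing boost that offsets the camera location, leaving rotation and FOV untouched.
; A timed intro eases toward a view-relative offset via a windup curve, then triggers forward decay on the boosting camera mode.
; The boost phase eases back along BoostCurve, removing the modifier when the curve ends.
; It adds forward offset when camera-to-target misaligns with view; both phases keep live Z and blend by modifier alpha.
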(event Custom|CheckEndModifier
  (bind (_mintime _maxtime) (Math|Curves|GetTimeRange (Variables|Default|GetBoostCurve)))
  (if (>= (Variables|Default|GetElapsedTime) _maxtime)
    (CallFunction|RemoveSelffromCameraManager)))

(event Custom|StartIntroTimer
  (Utilities|Time|SetTimerbyEvent (delegate Custom|EndIntro) (ScalableFloat|GetValueatLevel (Variables|Default|GetIntroDuration))))

(event Custom|EndIntro
  (Variables|Default|SetInIntro? false)
  (bind _asfort_player_camera_base (Utilities|Casting|CastToFortPlayerCameraBase (Variables|CameraModifier|GetCameraOwner)))
  (bind _ascamera_mode_cube_surfing_boosting (Utilities|Casting|CastToCameraMode_CubeSurfing_Boosting (Class|FortPlayerCameraBase|GetCurrentCameraMode _asfort_player_camera_base)))
  (Class|CameraModeCubeSurfingBoosting|StartForwardDecay _ascamera_mode_cube_surfing_boosting))

(fn BlueprintModifyCamera
  (params (DeltaTime :type Float%20%28single-precision%29) (ViewLocation :type Vector :default "0, 0, 0") (ViewRotation :type Rotator :default "0, 0, 0") (FOV :type Float%20%28single-precision%29))
  (bind (_viewlocation_x _viewlocation_y _viewlocation_z) (Math|Vector|BreakVector (Variables|GetViewLocation_X)))
  (bind (_cachedcurrentloc_x _cachedcurrentloc_y) (Math|Vector|BreakVector (Variables|GetCachedCurrentLoc_X)))
  (bind (_introoffset_x _introoffset_y _introoffset_z) (Math|Vector|BreakVector (Variables|GetIntroOffset_X)))
  (if (not (Variables|Default|GetInIntro?))
    (Utilities|FlowControl|Sequence
      (:then_0
        (if (not (Variables|Default|GetInBoost?))
          (Variables|Default|SetCachedStartLoc (Variables|Default|GetCachedCurrentLoc))
          (Variables|Default|SetCachedStartFOV (Variables|Default|GetCachedCurrentFOV))
          (Variables|Default|SetElapsedTime 0.0)
          (Variables|Default|SetInBoost? true)))
      (:then_1
        (CallFunction|LerpCamera (Variables|GetViewLocation) (Variables|GetFOV) (Variables|GetViewRotation) (Variables|GetDeltaTime))
        (CallFunction|TickElapsedTime (Variables|GetDeltaTime))
        (CallFunction|CheckEndModifier)
        (bind _finalloc (CallFunction|FinalDotModify (Math|Vector|MakeVector (.x (Variables|Default|GetCachedCurrentLoc)) (.y (Variables|Default|GetCachedCurrentLoc)) (.z (Variables|GetViewLocation))) (Variables|GetViewRotation)))
        (return (Math|Vector|Lerp(Vector) (Variables|GetViewLocation) _finalloc (Variables|CameraModifier|GetAlpha)) (Variables|GetViewRotation) (Variables|GetFOV))))
    (else
      (Utilities|FlowControl|Sequence
        (:then_0
          (if (not (Variables|Default|GetInitialized?))
            (CallFunction|StartIntroTimer)
            (Variables|Default|SetInitialized? true)
            (Variables|Default|SetCachedCurrentLoc (Variables|GetViewLocation))))
        (:then_1
          (Variables|Default|SetCachedCurrentLoc (Math|Vector|Lerp(Vector) (Variables|Default|GetCachedCurrentLoc) (Utilities|Operators|Add (Variables|GetViewLocation) (* (Math|Vector|GetForwardVector (Variables|GetViewRotation)) _introoffset_x) (* (Math|Vector|GetRightVector (Variables|GetViewRotation)) _introoffset_y) (Math|Vector|MakeVector 0.0 0.0 _introoffset_z)) (Math|Curves|GetFloatValue (Variables|Default|GetBoostWindupCurve) (Variables|Default|GetElapsedTime))))
          (CallFunction|TickElapsedTime (Variables|GetDeltaTime))
          (return (Math|Vector|Lerp(Vector) (Variables|GetViewLocation) (Math|Vector|MakeVector _cachedcurrentloc_x _cachedcurrentloc_y _viewlocation_z) (Variables|CameraModifier|GetAlpha)) (Variables|GetViewRotation) (Variables|GetFOV)))))))

(fn FinalDotModify
  (params (InLoc :type Vector) (InRot :type Rotator))
  (return (+ (* (Math|Vector|GetForwardVector InRot) (Math|Float|Lerp (Variables|Default|GetTurningMaxOffset) 0.0 (Math|Float|Absolute(Float) (Math|Vector|DotProduct (Math|Vector|Normalize2D(Vector) (- (Transformation|GetActorLocation (CameraModifier|GetViewTarget)) InLoc)) (Math|Vector|Normalize2D(Vector) (Math|Vector|GetForwardVector InRot)))))) InLoc)))

(fn GetTargetCameraRotation
  (return (Pawn|GetControlRotation (Fortnite|GetFortPlayerControllerfromActor (CameraModifier|GetViewTarget)) :!class /Script/Engine.Controller)))

(fn LerpCamera
  (params (InLoc :type Vector) (InFOV :type Float%20%28double-precision%29) (InRot :type Rotator) (DT :type Float%20%28double-precision%29))
  (Variables|Default|SetCachedCurrentLoc (Math|Vector|Lerp(Vector) (Variables|Default|GetCachedStartLoc) InLoc (Math|Curves|GetFloatValue (Variables|Default|GetBoostCurve) (Variables|Default|GetElapsedTime))))
  (return))

(fn TickElapsedTime
  (params (DT :type Float%20%28double-precision%29))
  (Variables|Default|SetElapsedTime (+ (Variables|Default|GetElapsedTime) DT)))
