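; Camera modifier for Cube Surfing that shifts camera location and FOV as the player surfs.
; Horizontal speed drives interpolated offsets to FOV and along control forward, while strong sideways acceleration adds a lateral offset, all blended by modifier alpha.
; HandleJump holds cached camera height for one second.
; PhaseOut eases the speed offsets to zero, drops the lateral one, and removes the modifier after its alpha-out time.
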(event Custom|InitModifier
  (bind _cameraowner (Variables|CameraModifier|GetCameraOwner))
  (bind _returnvalue (Game|Player|GetOwningPlayerController _cameraowner))
  (bind _returnvalue_2 (Pawn|GetControlledPawn _returnvalue))
  (bind _returnvalue_3 (Actor|GetComponentbyClass _returnvalue_2 "/Script/FortniteGame.FortMovementComp_Character"))
  (Variables|Default|SetPhasingOut? false)
  (Variables|Default|SetOwningPawn _returnvalue_2)
  (Utilities|IsValid _returnvalue_3
    (:Is%20Valid
      (Variables|Default|SetMoveComp _returnvalue_3))))

(event Custom|PhaseOut
  (Utilities|Time|SetTimerbyEvent (delegate Custom|EndModifier) (Variables|CameraModifier|GetAlphaOutTime))
  (Variables|Default|SetPhasingOut? true)
  (Variables|Default|SetDoBoostIntro? false)
  (Utilities|Time|ClearandInvalidateTimerbyHandle (Variables|Default|GetBoostIntroTimer)))

(event Custom|EndModifier
  (Variables|Default|SetPhasingOut? false)
  (CallFunction|RemoveSelffromCameraManager))

(event Custom|HandleJump
  (Variables|CacheVars|SetCachedLocation (Camera|GetCameraLocation (Variables|CameraModifier|GetCameraOwner)))
  (Variables|Default|SetIsJumping? true)
  (Utilities|FlowControl|Delay 1.0)
  (Variables|Default|SetIsJumping? false))

(fn BlueprintModifyCamera
  (params (DeltaTime :type Float%20%28single-precision%29) (ViewLocation :type Vector :default "0, 0, 0") (ViewRotation :type Rotator :default "0, 0, 0") (FOV :type Float%20%28single-precision%29))
  (bind _fov (Variables|GetFOV))
  (bind _deltatime (Variables|GetDeltaTime))
  (bind _deltatime_2 (Variables|GetDeltaTime))
  (bind _output_get (Variables|Default|SetPreModifiedCameraLoc ViewLocation))
  (if (not (Variables|Default|GetPhasingOut?))
    (Utilities|IsValid (Variables|Default|GetOwningPawn)
      (:Is%20Valid
        _)
      (:Is%20Not%20Valid
        (CallFunction|InitModifier)))
    (bind (_offset _fovoffset) (CallFunction|MakeXOffset _deltatime_2))
    (bind _offset_2 (Default|MakeAccelOffset _deltatime_2))
    (CallFunction|CacheFinalValues (Utilities|Operators|Add _offset (CallFunction|ModInitialLocation _output_get) _offset_2) (+ _fov _fovoffset))
    (return (Math|Vector|Lerp(Vector) (Variables|GetViewLocation) (Utilities|Operators|Add _offset (CallFunction|ModInitialLocation _output_get) _offset_2) (Variables|CameraModifier|GetAlpha)) (Variables|GetViewRotation) (Math|Float|Lerp (Variables|GetFOV) (+ _fov _fovoffset) (Variables|CameraModifier|GetAlpha)))
    (else
      (bind _output_get_2 (Variables|CacheVars|SetCachedOffset (Math|Interpolation|VInterpTo (Variables|CacheVars|GetCachedOffset) "0, 0, 0" _deltatime 2.0)))
      (bind _output_get_3 (Variables|CacheVars|SetCachedFOV (Math|Interpolation|FInterpTo (Variables|CacheVars|GetCachedFOV) 0.0 _deltatime 2.0)))
      (return (Math|Vector|Lerp(Vector) (Variables|GetViewLocation) (+ (CallFunction|ModInitialLocation _output_get) _output_get_2) (Variables|CameraModifier|GetAlpha)) (Variables|GetViewRotation) (Math|Float|Lerp (Variables|GetFOV) (+ (Variables|GetFOV) _output_get_3) (Variables|CameraModifier|GetAlpha))))))

(fn CacheFinalValues
  (params (CachedLocation :type Vector) (CachedFinalFOV :type Float%20%28double-precision%29))
  (if (not (Variables|Default|GetDoBoostIntro?))
    (Variables|CacheVars|SetCachedLocation CachedLocation)
    (Variables|CacheVars|SetCachedFinalFOV CachedFinalFOV)))

(fn CheckOverrideX
  (params (InOffset :type Vector))
  (return "0, 0, 0"))

(fn GetTargetCameraRotation
  (return (Pawn|GetControlRotation (Fortnite|GetFortPlayerControllerfromActor (CameraModifier|GetViewTarget)) :!class /Script/Engine.Controller)))

(fn MakeAccelOffset
  (params (DeltaTime :type Float%20%28double-precision%29))
  (bind _output_get (Variables|CacheVars|SetCachedAccelOffset (Math|Interpolation|VInterptoConstant (Variables|CacheVars|GetCachedAccelOffset) (Math|Vector|SelectVector (* (Math|Vector|SelectVector (Math|Vector|NegateVector (Math|Vector|Normalize2D(Vector) (Math|Vector|GetRightVector (Pawn|GetBaseAimRotation (Variables|Default|GetOwningPawn))))) (Math|Vector|Normalize2D(Vector) (Math|Vector|GetRightVector (Pawn|GetBaseAimRotation (Variables|Default|GetOwningPawn)))) (< (Math|Vector|DotProduct (Math|Vector|Normalize2D(Vector) (Pawn|Components|CharacterMovement|GetCurrentAcceleration (Variables|Default|GetMoveComp))) (Math|Vector|Normalize2D(Vector) (Math|Vector|GetRightVector (Pawn|GetBaseAimRotation (Variables|Default|GetOwningPawn))))) 0)) (Variables|Default|GetMaxYOffset)) "0, 0, 0" (> (Math|Float|Absolute(Float) (Math|Vector|DotProduct (Math|Vector|Normalize2D(Vector) (Pawn|Components|CharacterMovement|GetCurrentAcceleration (Variables|Default|GetMoveComp))) (Math|Vector|Normalize2D(Vector) (Math|Vector|GetRightVector (Pawn|GetBaseAimRotation (Variables|Default|GetOwningPawn)))))) 0.5)) (Variables|GetDeltaTime) (Math|Float|Clamp(Float) (Math|Vector|VectorLength (Math|Vector|VectorGetAbs (- (Math|Vector|SelectVector (* (Math|Vector|SelectVector (Math|Vector|NegateVector (Math|Vector|Normalize2D(Vector) (Math|Vector|GetRightVector (Pawn|GetBaseAimRotation (Variables|Default|GetOwningPawn))))) (Math|Vector|Normalize2D(Vector) (Math|Vector|GetRightVector (Pawn|GetBaseAimRotation (Variables|Default|GetOwningPawn)))) (< (Math|Vector|DotProduct (Math|Vector|Normalize2D(Vector) (Pawn|Components|CharacterMovement|GetCurrentAcceleration (Variables|Default|GetMoveComp))) (Math|Vector|Normalize2D(Vector) (Math|Vector|GetRightVector (Pawn|GetBaseAimRotation (Variables|Default|GetOwningPawn))))) 0)) (Variables|Default|GetMaxYOffset)) "0, 0, 0" (> (Math|Float|Absolute(Float) (Math|Vector|DotProduct (Math|Vector|Normalize2D(Vector) (Pawn|Components|CharacterMovement|GetCurrentAcceleration (Variables|Default|GetMoveComp))) (Math|Vector|Normalize2D(Vector) (Math|Vector|GetRightVector (Pawn|GetBaseAimRotation (Variables|Default|GetOwningPawn)))))) 0.5)) (Variables|CacheVars|GetCachedAccelOffset)))) 1e+02 5e+02))))
  (return _output_get))

(fn MakeXOffset
  (params (DeltaTime :type Float%20%28double-precision%29))
  (bind _owningpawn (Variables|Default|GetOwningPawn))
  (bind _returnvalue (Transformation|GetVelocity _owningpawn))
  (bind _returnvalue_2 (Math|Vector|VectorLengthXY _returnvalue))
  (bind _minspeed (LocalVariables|GetMinSpeed))
  (bind _deltatime (Variables|GetDeltaTime))
  (bind _maxspeed (LocalVariables|GetMaxSpeed))
  (Utilities|IsValid _owningpawn
    (:Is%20Valid
      (bind _output_get (Variables|CacheVars|SetCachedOffset (Math|Interpolation|VInterptoConstant (Variables|CacheVars|GetCachedOffset) (* (Math|Vector|GetForwardVector (Pawn|GetControlRotation _owningpawn :!class /Script/Engine.Pawn)) (Math|Float|MapRangeClamped _returnvalue_2 _minspeed _maxspeed 0.0 (Variables|Default|GetMaxXOffset))) _deltatime (Variables|Default|GetXInterpSpeed))))
      (bind _output_get_2 (Variables|CacheVars|SetCachedFOV (Math|Interpolation|FInterptoConstant (Variables|CacheVars|GetCachedFOV) (Math|Float|MapRangeClamped _returnvalue_2 _minspeed _maxspeed 0.0 (Variables|Default|GetMaxFOVOffset)) _deltatime 15.0)))
      (return _output_get _output_get_2))
    (:Is%20Not%20Valid
      (return "0, 0, 0" 0.0))))

(fn ModInitialLocation
  (params (InLocation_X :type Float%20%28double-precision%29) (InLocation_Y :type Float%20%28double-precision%29) (InLocation_Z :type Float%20%28double-precision%29))
  (bind (_cachedlocation_x _cachedlocation_y _cachedlocation_z) (Math|Vector|BreakVector (Variables|GetCachedLocation_X)))
  (return InLocation_X InLocation_Y (Math|Float|SelectFloat _cachedlocation_z InLocation_Z (Variables|Default|GetIsJumping?))))
