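; Camera modifier for CubeSurfing that, going by its name, plays a camera offset effect when surfing starts.
; Each update accumulates elapsed time and samples a float curve to blend view location toward forward and right offsets relative to view rotation.
; The same curve value blends FOV toward an added offset, while rotation passes through unchanged.
; A separate helper, not called here, returns the view target's Fortnite player controller rotation.
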(fn BlueprintModifyCamera
  (params (DeltaTime :type Float%20%28single-precision%29) (ViewLocation :type Vector :default "0, 0, 0") (ViewRotation :type Rotator :default "0, 0, 0") (FOV :type Float%20%28single-precision%29))
  (bind _output_get (Variables|Default|SetTimeElapsed (+ DeltaTime (Variables|Default|GetTimeElapsed))))
  (return (Math|Vector|Lerp(Vector) ViewLocation (Utilities|Operators|Add ViewLocation (* (Math|Vector|GetForwardVector ViewRotation) (Variables|Default|GetForwardOffset)) (* (Math|Vector|GetRightVector ViewRotation) (Variables|Default|GetXYOffset))) (Math|Curves|GetFloatValue (Variables|Default|GetCurve) _output_get)) ViewRotation (Math|Float|Lerp FOV (+ FOV (Variables|Default|GetFOVOffset)) (Math|Curves|GetFloatValue (Variables|Default|GetCurve) _output_get))))

(fn GetTargetCameraRotation
  (return (Pawn|GetControlRotation (Fortnite|GetFortPlayerControllerfromActor (CameraModifier|GetViewTarget)) :!class /Script/Engine.Controller)))
